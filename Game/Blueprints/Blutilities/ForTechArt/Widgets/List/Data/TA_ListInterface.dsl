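; TA_ListInterface is an interface for list items in the tech art editor utility widgets.
; It declares functions to pass an item its data object, its selection state and its owning list widget.
; FilterBySearchString takes a search string and returns a Boolean, giving each item a hook for search filtering.
; The functions hold no logic here, so implementing widgets supply the behavior.
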(fn FilterBySearchString
  (params (SearchString :type String))
  (return false))

(fn setListItemData
  (params (listItemData :type Object%20Reference)))

(fn setListItemSelection
  (params (selected :type Boolean)))

(fn setOwningList
  (params (list :type User%20Widget%20Object%20Reference)))
